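# Parent material for the Total Flock pickaxe, inserting a reactive dragon emissive between the 2023 character PreFX and PostFX functions.
# A packed mask texture supplies horn, emissive, reveal-gradient and eye masks for a composite of fresnel outlines, flow and sine pulsing, tinted by a three-color lerp.
# Two damage start times drive summed ramp-up/down envelopes feeding a smoothstep reveal mask, with Time Step as the alternative when reactivity is disabled.
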Material /BRCosmetics/Weapons/FORT_Melee/Pickaxe_TotalFlock/Materials/M_Pickaxe_TotalFlock_Parent
graph {
 properties {
  bUseMaterialAttributes=true
  bUsedWithMeshDeformer=true
  bUsedWithMorphTargets=true
  bUsedWithNanite=true
  bUsedWithNiagaraMeshParticles=true
  bUsedWithSkeletalMesh=true
  pixelDepthOffsetMode=PDOM_Legacy
  shadingModel=MSM_FromMaterialExpression
  usedShadingModels=Default Lit | Subsurface | Clear Coat | Subsurface Profile | Cloth
 }
 nodes [
  n1 = MaterialFunctionCall() {materialFunction=&MF_CharacterParent_2023_PreFX}
  n2 = GetMaterialAttributes(None=@n1.Result) {attributeGetTypes [#1, #2]}
  MaterialFunctionCall(
   Material=SetMaterialAttributes!(MaterialAttributes=@n2.MaterialAttributes, Emissive Color=NamedRerouteUsage(), Base Color=@n2.MaterialAttributes) {attributeSetTypes [#1, #2]}
   M=@n1.M
   Pre Skinned Local Position=@n1.M
   Distance Blend=@n1.M
   Fresnel=@n1.M
  ) {materialFunction=&MF_CharacterParent_2023_PostFX}
  n3 = Constant()
  n4 = VectorParameter() {group=Total Flock - Pickaxe, parameterName=Dragon Emissive 1, defaultValue=<rgba 0 0.5822 1.5652 1>}
  n5 = VectorParameter() {group=Total Flock - Pickaxe, parameterName=Dragon Emissive 2, defaultValue=<rgba 0 1.1508 1.5652 1>}
  n6 = NamedRerouteUsage()
  NamedRerouteDeclaration(None=Multiply(
    A=MaterialFunctionCall(A=@n3, B=@n4.RGB, C=@n5.RGB, Alpha=@n6) {materialFunction=&Lerp_3Color}.Result
    B=ScalarParameter() {defaultValue=1, group=Total Flock - Pickaxe, parameterName=Emissive Intensity}
   )) {name=Emissive, nodeColor=<rgba 0 1 0.5781 1>}
  n7 = NamedRerouteUsage()
  n8 = Multiply(A=NamedRerouteUsage(), B=OneMinus(None=NamedRerouteUsage())) {desc=Clip horn to only show staff fresnel.}
  n9 = NamedRerouteUsage()
  n10 = Add(A=@n7, B=Saturate(None=Multiply(A=@n8, B=@n9)))
  n11 = NamedRerouteUsage()
  n12 = NamedRerouteUsage()
  n13 = Multiply(
   A=Power(
    Base=Multiply(A=@n11, B=@n12)
    Exp=ScalarParameter() {defaultValue=0.552, group=Total Flock - Pickaxe, parameterName=Horn - Fresnel Fade Strength}
   )
   B=ScalarParameter() {defaultValue=2.2255, group=Total Flock - Pickaxe, parameterName=Horn - Brightness}
  )
  n14 = NamedRerouteUsage()
  n15 = NamedRerouteUsage()
  n16 = Max(
   A=Max(A=Add(A=@n10, B=@n13), B=Multiply(A=@n14, B=@n15))
   B=Multiply(A=NamedRerouteUsage(), B=ScalarParameter() {defaultValue=0.15, group=Total Flock - Pickaxe, parameterName=Horn - Min Value})
  )
  n17 = MaterialFunctionCall!(
   Sine Phase=Multiply(A=Time(), B=ScalarParameter() {defaultValue=0.1, group=Total Flock - Pickaxe, parameterName=Sine Speed})
   Value 1=ScalarParameter() {defaultValue=0.8, group=Total Flock - Pickaxe, parameterName=Staff - Emissive Intensity - Min}
   Value 2=ScalarParameter() {defaultValue=2, group=Total Flock - Pickaxe, parameterName=Staff - Emissive - Intensity Max}
  ) {materialFunction=/Engine/Functions/Engine_MaterialFunctions02/Utility/Sine_Remapped}
  n18 = Subtract(A=NamedRerouteUsage(), B=NamedRerouteUsage()) {desc=Clip out the horn from the emissive mask.}
  NamedRerouteDeclaration(None=Max(A=Subtract(A=@n18, B=@n7), B=Multiply(A=@n16, B=@n17.Remapped Sine))) {name=Emissive Composite, nodeColor=<rgba 1 0.2578 0 1>}
  n19 = TextureSample() {samplerType=SAMPLERTYPE_Masks, texture=/BRCosmetics/Weapons/FORT_Melee/Pickaxe_TotalFlock/Textures/T_TotalFlock_Pickaxe_M}
  NamedRerouteDeclaration(None=@n19.B) {name="Reveal Gradient (B)", nodeColor=<rgba 0 1 0.3906 1>}
  n20 = NamedRerouteUsage()
  n21 = ScalarParameter() {defaultValue=0.3, group=Total Flock - Pickaxe, parameterName=Staff - Reveal - Edge Smoothness}
  n22 = ScalarParameter() {defaultValue=1, group=Total Flock - Pickaxe, parameterName=Time Step}
  n23 = Subtract(A=Time(), B=ScalarParameter() {defaultValue=-100, group=Total Flock - Pickaxe, parameterName=DmgReactivityStartTime_1})
  n24 = Saturate(None=Divide(A=@n23, B=ScalarParameter() {defaultValue=0.4, group=Total Flock - Pickaxe, parameterName=ReactivityRampUpDuration}) {constB=0.8})
  n25 = ScalarParameter() {defaultValue=0.75, group=Total Flock - Pickaxe, parameterName=ReactivityRampUpExp}
  n26 = ScalarParameter() {defaultValue=10, group=Total Flock - Pickaxe, parameterName=ReactivityRampDownDuration}
  n27 = OneMinus(None=Saturate(None=Divide(A=@n23, B=@n26) {constB=0.8}))
  n28 = ScalarParameter() {defaultValue=1, group=Total Flock - Pickaxe, parameterName=ReactivityRampDownExp}
  n29 = Subtract(A=Time(), B=ScalarParameter() {defaultValue=-100, group=Total Flock - Pickaxe, parameterName=DmgReactivityStartTime_0})
  n30 = Saturate(None=Divide(A=@n29, B=ScalarParameter() {defaultValue=0.4, group=Total Flock - Pickaxe, parameterName=ReactivityRampUpDuration}) {constB=0.8})
  n31 = ScalarParameter() {defaultValue=0.75, group=Total Flock - Pickaxe, parameterName=ReactivityRampUpExp}
  n32 = ScalarParameter() {defaultValue=10, group=Total Flock - Pickaxe, parameterName=ReactivityRampDownDuration}
  n33 = OneMinus(None=Saturate(None=Divide(A=@n29, B=@n32) {constB=0.8}))
  n34 = ScalarParameter() {defaultValue=1, group=Total Flock - Pickaxe, parameterName=ReactivityRampDownExp}
  n35 = Add(
   A=Multiply(A=Power(Base=@n24, Exp=@n25) {constExponent=0.8}, B=Power(Base=@n27, Exp=@n28) {constExponent=0.4})
   B=Multiply(A=Power(Base=@n30, Exp=@n31) {constExponent=0.8}, B=Power(Base=@n33, Exp=@n34) {constExponent=0.4})
  )
  n36 = Add(
   A=Multiply(A=@n35, B=ScalarParameter() {defaultValue=0.4, group=Total Flock - Pickaxe, parameterName=ReactivitySingleHitStrength})
   B=Multiply(A=Saturate(None=@n35), B=ScalarParameter() {defaultValue=0.7, group=Total Flock - Pickaxe, parameterName=ReactivityFirstHitStrengthBoost}) {constB=0.4}
  )
  n37 = ScalarParameter() {defaultValue=1, group=Total Flock - Pickaxe, parameterName=Enable Reactivity}
  NamedRerouteDeclaration(None=SmoothStep(Min=@n20, Max=Add(A=@n20, B=@n21), Value=LinearInterpolate(A=@n22, B=@n36, Alpha=@n37))) {name=Reveal Mask, nodeColor=<rgba 1 0.0703 0 1>}
  NamedRerouteDeclaration(None=@n2.Emissive Color) {name=Emissive - In, nodeColor=<rgba 0 0.3437 1 1>}
  n38 = NamedRerouteUsage()
  n39 = NamedRerouteUsage()
  n40 = NamedRerouteUsage()
  n41 = NamedRerouteUsage()
  NamedRerouteDeclaration(None=LinearInterpolate(A=NamedRerouteUsage(), B=Add(A=@n38, B=@n39), Alpha=LinearInterpolate(B=@n40, Alpha=@n41) {constAlpha=0})) {name=Emissive - Out, nodeColor=<rgba 0.6641 1 0 1>}
  NamedRerouteDeclaration(None=@n19.A) {name="Eyes Mask (A)", nodeColor=<rgba 0 1 0.3906 1>}
  n42 = MaterialFunctionCall!(FlowID, FlowGradient=@n19.B, FlowMask=NamedRerouteUsage()) {materialFunction=/Game/Global/Materials/MaterialGym/Materials/Flow/MaterialFunctions/MF_Flow}
  n43 = NamedRerouteUsage()
  n44 = OneMinus(None=Max(A=NamedRerouteUsage(), B=@n43))
  NamedRerouteDeclaration(None=Multiply(A=Multiply(A=@n42.Result, B=@n44), B=Step(Y=@n43) {constX=0.1})) {name=Flow, nodeColor=<rgba 1 0.0703 0 1>}
  NamedRerouteDeclaration(None=@n19.R) {name="Horn (R)", nodeColor=<rgba 0.2422 1 0 1>}
  NamedRerouteDeclaration(None=Power(
    Base=MaterialFunctionCall() {materialFunction=&BoundingBoxBased_0-1_UVW}.B
    Exp=ScalarParameter() {defaultValue=3.164, group=Total Flock - Pickaxe, parameterName=Staff - Top/Down Gradient}
   )) {name=Staff - Top/Down Gradient, nodeColor=<rgba 1 0 0.5625 1>}
  NamedRerouteDeclaration(None=@n19.G) {name="Emissive Mask (G)", nodeColor=<rgba 0 0.8828 1 1>}
  n45 = MaterialFunctionCall(
   Invert Fresnel=StaticBoolParameter() {defaultValue=true, parameterName="Invert Fresnel (SB)"}
   Power=ScalarParameter() {defaultValue=3.5754, group=Total Flock - Pickaxe, parameterName=Horn - Fresnel - Fade}
  ) {materialFunction=&Fresnel_Function}
  n46 = NamedRerouteUsage()
  NamedRerouteDeclaration(None=Multiply(
    A=Multiply(A=@n45.Result, B=@n46)
    B=ScalarParameter() {defaultValue=0.2479, group=Total Flock - Pickaxe, parameterName=Horn - Fresnel - Strength}
   )) {name=Horn Fresnel, nodeColor=<rgba 0.4766 0 1 1>}
  n47 = Fresnel(ExponentIn=ScalarParameter() {defaultValue=1.9155, group=Total Flock - Pickaxe, parameterName=Staff - Fresnel}) {baseReflectFraction=0}
  n48 = OneMinus(None=@n47)
  n49 = ScalarParameter() {group=Total Flock - Pickaxe, parameterName=Staff - Outline - Step Amount}
  NamedRerouteDeclaration(None=Max(A=Step(Y=@n48, X=@n49) {constX=0.7}, B=@n47)) {name=Staff - Fresnel, nodeColor=<rgba 0 1 0.1562 1>}
  n50 = NamedRerouteUsage()
  n51 = ScalarParameter() {defaultValue=0.9215, group=Total Flock - Pickaxe, parameterName=Horn - Outline - Fade}
  n52 = Step(Y=@n48, X=ScalarParameter() {defaultValue=0.8, group=Total Flock - Pickaxe, parameterName=Horn - Outline - Step Amount}) {constX=0.7}
  n53 = ScalarParameter() {defaultValue=1, group=Total Flock - Pickaxe, parameterName=Horn - Outline - Fresnel Strength}
  NamedRerouteDeclaration(None=Multiply(A=Power(Base=@n50, Exp=@n51), B=Multiply(A=@n52, B=@n53))) {name=Horn - Fresnel Outline, nodeColor=<rgba 1 0 0.1641 1>}
 ]
}
